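# Niagara burst effect for the Dazzle backpack cosmetic, spawning toon-shaded, outlined mesh particles according to User.KillStage.
# Stage-gated emitters pair main meshes, animated by time-driven CustomMeshIndex frames that User.TriggerKill switches, with plus and star accents at higher stages.
# Time-indexed offset and scale arrays add stepped jitter, and mostly constant-key curves pop meshes before shrinking them away.
# User.VariantID selects visibility tags or colors, and time-of-day brightness scales color.
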NiagaraSystem /BRCosmetics/Accessories/FORT_Backpacks/Backpack_F_MED_Dazzle/FX/NS_Backpack_Dazzle_Burst02
dependencies {
 usedDataInterfaces [
  &NiagaraDataInterfaceArrayFloat3
  &NiagaraDataInterfaceArrayFloat
  &NiagaraDataInterfaceCurve
  &NiagaraDataInterfacePlatformSet
  &NiagaraDataInterfaceArrayInt32
 ]
 usedDynamicInputs [/Niagara/DynamicInputs/Math/Floor_Float, &FNTimeOfDayScalar]
 usedModules [
  &self:SystemSpawnScript.SystemScriptSource.SystemScriptGraph.NiagaraNodeAssignment_0.SetVariables_#1
  &SystemState
  &self:SystemSpawnScript.SystemScriptSource.SystemScriptGraph.NiagaraNodeAssignment_1.SetVariables_#2
  &self:SystemSpawnScript.SystemScriptSource.SystemScriptGraph.NiagaraNodeAssignment_3.SetVariables_#3
  &self:SystemSpawnScript.SystemScriptSource.SystemScriptGraph.NiagaraNodeAssignment_2.SetVariables_#4
 ]
}
emitters {
 NoFixedRate001 {
  emitterSpawnScript [/SetVariables(Emitter.FramePerShape=8, Emitter.ShapeNumber=1) {isSetParameters=true, module=SetVariables_#5}]
  emitterUpdateScript [
   &EmitterState(Life Cycle Mode=::NewEnumerator1, Loop Behavior=::NewEnumerator1, Loop Duration Mode=::NewEnumerator1, UseLoopDelay=false)
   &SpawnBurst_Instantaneous(Spawn Group=1, Use Spawn Probability=false, UseLoopCountLimit=false) {enabled=false}
   &SpawnBurst_Instantaneous(
    Spawn Count=&MakeCustomIntFromBool(Bool=&SetBoolByIntComparison(A=$User.KillStage, B=2, Comparison Type=::NewEnumerator4), True INT=1, False INT=0)
    Spawn Group=1
    Use Spawn Probability=false
    UseLoopCountLimit=false
   ) {module=SpawnBurst_Instantaneous001}
   /Game/Effects/Niagara/MasterMaterials/ToonShader/Template/NM_ToonShaderParameters(
    HighLight_Softness=1
    Lighting_Hardness=0
    Metallic_LightingPower=0
    RimLight_Color:LinearColor=<rgba 0.1615 0.1615 0.1615 1>
    RimLight_Width=0
   )
   /Game/Effects/Niagara/MasterMaterials/ToonShader/Template/NM_OutlineParameters(OutlineDivisorOffset=4)
  ]
  particleSpawnScript [
   &FN_SetSpawnGroup(Number of Spawn Groups=2)
   &InitializeParticle(
    Lifetime=1.5
    Mesh Scale Mode=::NewEnumerator4
    Sprite Size Mode=::NewEnumerator3
    Sprite UV Mode=::NewEnumerator3
    Uniform Sprite Size=25
    UsePositionOffset=false
    Write Material Random=true
   ) {module=InitializeParticle001}
   /SetVariables() {isSetParameters=true}
   /SetVariables(Particles.CustomMeshIndex:NiagaraInt32=&RandomRangeInt(Minimum=0, Maximum=3)) {isSetParameters=true, module=SetVariables_#6}
   &SkeletalMeshLocation(
    Bone / Socket Index=&SelectIntFromArray(Direct Array Index=&SpawnGroup(), Int Selection Array=DataInterface() {propertyValues {IntData [0,1]}})
    MaskSpawnGroup=false
    Mesh Sampling Type=::NewEnumerator5
    Sampled Orientation Coordinate Space=::World
    Skeletal Mesh=$User.SkeletalMesh
    Socket Sampling Mode=::NewEnumerator5
   ) {enabled=false, module=SkeletalMeshLocation001}
   /SetVariables(Particles.VisibilityTag=&MakeIntFromFloat(Float=$User.VariantID)) {isSetParameters=true}
   /SetVariables(Particles.Color=&MultiplyLinearColor(
     Linear Color=$Particles.Color
     Scale Factor=&MakeVector4_FromLinearColor(LinearColor=/Niagara/DynamicInputs/TypeConversions/MakeLinearColorFromFloat(Value:NiagaraFloat=$System.TimeOfDayBrightnessMultiplier))
    )) {isSetParameters=true}
  ]
  particleUpdateScript [
   &ParticleState()
   &ScaleMeshSize(Scale Factor=&MakeVector(
     X=1
     Y=&FloatFromCurve(FloatCurve=DataInterface() {propertyValues {Curve {keys [<at 0 2 RCIM_Constant>,<at 0.05 1.5 RCIM_Constant>,<at 0.1 1 RCIM_Constant>]}}})
     Z=1
    )) {enabled=false}
   &SkeletalMeshLocation(
    Bone / Socket Index=&SelectIntFromArray(Direct Array Index=&SpawnGroup(), Int Selection Array=DataInterface() {propertyValues {IntData [0,1]}})
    MaskSpawnGroup=false
    Mesh Sampling Type=::NewEnumerator5
    Sampled Orientation Coordinate Space=::World
    Skeletal Mesh=$User.SkeletalMesh
    Socket Sampling Mode=::NewEnumerator5
    Velocity Sampling=::NewEnumerator0
    VelocitySpeedLimit=false
   ) {enabled=false}
   /SetVariables(Particles.Position=$Particles.SkeletalMeshLocation.SampledPosition) {enabled=false, isSetParameters=true, module=SetVariables_#7}
   /SetVariables(
    Particles.Position=&AddVectorToPosition(
     Position=$Particles.Initial.Position
     Vector=&Multiply_Vector(
      A=&VectorFromFloat(Value=0)
      B=&SelectVectorFromArray(
       Direct Array Index=&ModuloInt(A=&MakeIntFromFloat(Float=&Multiply_Float(A=$Engine.Time, B=6)), B=6)
       Vector Selection Array=DataInterface() {propertyValues {FloatData [<1 0 0>, <0 -0.5 -0.5>, <-1 0.5 0.5>, <0.5 2 1>, <0.1 -0.4 0.2>, <0.35 0.7 -0.4>]}}
      )
     )
    )
    Particles.Scale=&Multiply_Vector(
     A=$Particles.Initial.Scale
     B=&Multiply_Vector(
      A=<1 1 1>
      B=&SelectVectorFromArray(
       Direct Array Index=&ModuloInt(A=&MakeIntFromFloat(Float=&Multiply_Float(A=$Engine.Time, B=6)), B=3)
       Vector Selection Array=DataInterface() {propertyValues {FloatData [<1.02 1 1>, <1.04 1.07 1.05>, <1 1.02 0.99>]}}
      )
     )
    )
   ) {isSetParameters=true, module=SetVariables_#8}
   &ScaleMeshSize(Scale Factor=&VectorFromFloat(Value=&FloatFromCurve(FloatCurve=DataInterface() {propertyValues {Curve {keys [{tangentMode=RCTM_User, time=0, value=1}, <at 0.7 1 RCIM_Constant>, <at 0.8 0.6 RCIM_Constant>, <at 0.9 -0.1>]}}}))) {enabled=false, module=ScaleMeshSize002}
   /SetVariables(Particles.MeshOrientation=/Niagara/DynamicInputs/TypeConversions/MakeCustomQuaternionFromBool(
     Bool:NiagaraBool=$System.SteppedTickRate.RefreshTick
     True Quat:Quat4f=$Particles.SkeletalMeshLocation.SampledRotation
     False Quat:Quat4f=$Particles.MeshOrientation
    )) {enabled=false, isSetParameters=true, module=SetVariables_#9}
   /SetVariables(Particles.CustomMeshIndex:NiagaraInt32=&MakeCustomIntFromBool(
     Bool=$User.TriggerKill
     True INT=&Add_Integer(
      A=&Multiply_Int(A=$Emitter.FramePerShape, B=$Emitter.ShapeNumber)
      B=&ModuloInt(
       A=&MakeIntFromFloat(Float=&Multiply_Float(A=$Engine.Time, B=8))
       B=$Emitter.FramePerShape
      )
     )
     False INT=&ModuloInt(
      A=&Add_Integer(
       A=&ModuloInt(
        A=&MakeIntFromFloat(Float=&Multiply_Float(A=$Engine.Time, B=6))
        B=$Emitter.FramePerShape
       )
       B=&Multiply_Int(
        A=&ModuloInt(
         A=&Add_Integer(
          A=&MakeIntFromFloat(Float=&Multiply_Float(A=$Engine.Time, B=1))
          B=&ReturnExecIndex()
         )
         B=$Emitter.ShapeNumber
        )
        B=$Emitter.FramePerShape
       )
      )
      B=&Multiply_Int(A=$Emitter.FramePerShape, B=$Emitter.ShapeNumber)
     )
    )) {isSetParameters=true, module=SetVariables_#10}
   &ScaleColor(
    Color Value To Scale=$Particles.Color
    Scale RGB=&Multiply_VectorByFloat(Vector=<1 1 1>, Float=1)
    ScaleA=true
    ScaleRGB=true
   ) {module=ScaleColor001}
   &ScaleMeshSize(
    Initial Mesh Scale:Vector3f=$StackContext.Scale
    Scale Factor=&MakeVector(
     X=1
     Y=1
     Z=&FloatFromCurve(
      FloatCurve=DataInterface() {propertyValues {Curve {keys [
          {tangentMode=RCTM_User, time=0, value=1}
          <at 0.9 0.95 RCIM_Constant>
          <at 0.95 1.6 RCIM_Constant>
          <at 0.98 0.52 RCIM_Constant>
          <at 1 -0.1>
         ]}}}
      CurveIndex=&Subtract_Float(
       A=$Particles.NormalizedAge
       B=&ReturnNormalizedExecIndex(Normalized Index Scale=0.15, Normalized Index Includes One=false)
      )
     )
    )
   ) {module=ScaleMeshSize001}
   &Color(Color=&SelectLinearColorFromArray(
     Array Sampling Mode=::NewEnumerator1
     Direct Array Index=&Add_Integer(
      A=&MakeIntFromFloat(Float=&Multiply_Float(A=$Engine.Time, B=1))
      B=&ReturnExecIndex()
     )
     Direct Array Mode=::NewEnumerator1
     Color Selection Array=DataInterface() {propertyValues {ColorData [<rgba 1 0.1587 0.2417 1>, <rgba 0.949 0.3176 1 1>, <rgba 0.342 0.5018 1 1>, <rgba 1 0.8519 0.2575 1>, <rgba 0.3783 1 0.3917 1>]}}
     Random Seed=?
    )) {enabled=false}
  ]
  rendererClasses [&NiagaraMeshRendererProperties]
  renderers [
   {rendererClass=&NiagaraMeshRendererProperties, rendererIndex=0}
   {rendererClass=&NiagaraMeshRendererProperties, rendererIndex=1}
   {rendererClass=&NiagaraMeshRendererProperties, rendererIndex=2}
  ]
 }
 NoFixedRate002 {
  emitterSpawnScript [/SetVariables(Emitter.FramePerShape=8, Emitter.ShapeNumber=1) {isSetParameters=true, module=SetVariables_#5}]
  emitterUpdateScript [
   &EmitterState(Life Cycle Mode=::NewEnumerator1, Loop Behavior=::NewEnumerator1, Loop Duration Mode=::NewEnumerator1, UseLoopDelay=false)
   &SpawnBurst_Instantaneous(Spawn Group=1, Use Spawn Probability=false, UseLoopCountLimit=false) {enabled=false}
   &SpawnBurst_Instantaneous(
    Spawn Count=&MakeCustomIntFromBool(Bool=&SetBoolByIntComparison(A=$User.KillStage, B=3, Comparison Type=::NewEnumerator4), True INT=1, False INT=0)
    Spawn Group=1
    Use Spawn Probability=false
    UseLoopCountLimit=false
   ) {module=SpawnBurst_Instantaneous001}
   &NM_ToonShaderParameters(Lighting_Hardness=0, Metallic_LightingPower=0, RimLight_Color:LinearColor=<rgba 0.1615 0.1615 0.1615 1>, RimLight_Width=0)
   &NM_OutlineParameters(OutlineDivisorOffset=4)
  ]
  particleSpawnScript [
   &FN_SetSpawnGroup(Number of Spawn Groups=2)
   &InitializeParticle(
    Lifetime=1.7
    Mesh Scale Mode=::NewEnumerator4
    Sprite Size Mode=::NewEnumerator3
    Sprite UV Mode=::NewEnumerator3
    Uniform Sprite Size=25
    UsePositionOffset=false
    Write Material Random=true
   ) {module=InitializeParticle001}
   /SetVariables() {isSetParameters=true}
   /SetVariables(Particles.CustomMeshIndex:NiagaraInt32=&RandomRangeInt(Minimum=0, Maximum=3)) {isSetParameters=true, module=SetVariables_#6}
   &SkeletalMeshLocation(
    Bone / Socket Index=&SelectIntFromArray(Direct Array Index=&SpawnGroup(), Int Selection Array=DataInterface() {propertyValues {IntData [0,1]}})
    MaskSpawnGroup=false
    Mesh Sampling Type=::NewEnumerator5
    Sampled Orientation Coordinate Space=::World
    Skeletal Mesh=$User.SkeletalMesh
    Socket Sampling Mode=::NewEnumerator5
   ) {enabled=false, module=SkeletalMeshLocation001}
   /SetVariables(Particles.VisibilityTag=&MakeIntFromFloat(Float=$User.VariantID)) {isSetParameters=true}
   /SetVariables(Particles.New Vector:Vector3f=<1 1 1>) {isSetParameters=true}
   /SetVariables(Particles.Color=&MultiplyLinearColor(
     Linear Color=$Particles.Color
     Scale Factor=&MakeVector4_FromFloat(Float=$System.TimeOfDayBrightnessMultiplier)
    )) {isSetParameters=true}
  ]
  particleUpdateScript [
   &ParticleState()
   &ScaleMeshSize(Scale Factor=&MakeVector(
     X=1
     Y=&FloatFromCurve(FloatCurve=DataInterface() {propertyValues {Curve {keys [<at 0 2 RCIM_Constant>,<at 0.05 1.5 RCIM_Constant>,<at 0.1 1 RCIM_Constant>]}}})
     Z=1
    )) {enabled=false}
   &SkeletalMeshLocation(
    Bone / Socket Index=&SelectIntFromArray(Direct Array Index=&SpawnGroup(), Int Selection Array=DataInterface() {propertyValues {IntData [0,1]}})
    MaskSpawnGroup=false
    Mesh Sampling Type=::NewEnumerator5
    Sampled Orientation Coordinate Space=::World
    Skeletal Mesh=$User.SkeletalMesh
    Socket Sampling Mode=::NewEnumerator5
    Velocity Sampling=::NewEnumerator0
    VelocitySpeedLimit=false
   ) {enabled=false}
   /SetVariables(Particles.Position=$Particles.SkeletalMeshLocation.SampledPosition) {enabled=false, isSetParameters=true, module=SetVariables_#7}
   /SetVariables(
    Particles.Position=&AddVectorToPosition(
     Position=$Particles.Initial.Position
     Vector=&Multiply_Vector(
      A=&VectorFromFloat(Value=0.4)
      B=&SelectVectorFromArray(
       Direct Array Index=&ModuloInt(A=&MakeIntFromFloat(Float=&Multiply_Float(A=$Engine.Time, B=6)), B=6)
       Vector Selection Array=DataInterface() {propertyValues {FloatData [<1 0 0>, <0 -0.5 -0.5>, <-1 0.5 0.5>, <0.5 2 1>, <0.1 -0.4 0.2>, <0.35 0.7 -0.4>]}}
      )
     )
    )
    Particles.Scale=&Multiply_Vector(
     A=$Particles.Initial.Scale
     B=&SelectVectorFromArray(
      Direct Array Index=&ModuloInt(A=&MakeIntFromFloat(Float=&Multiply_Float(A=$Engine.Time, B=6)), B=3)
      Vector Selection Array=DataInterface() {propertyValues {FloatData [<1 1 1>, <1.05 1 1.05>, <0.95 1.02 0.99>]}}
     )
    )
   ) {isSetParameters=true, module=SetVariables_#8}
   &ScaleMeshSize(Scale Factor=&VectorFromFloat(Value=&FloatFromCurve(FloatCurve=DataInterface() {propertyValues {Curve {keys [
          {tangentMode=RCTM_User, time=0, value=1}
          {interpMode=RCIM_Cubic, tangentMode=RCTM_User, time=0.85, value=1}
          {interpMode=RCIM_Cubic, tangentMode=RCTM_User, time=0.9, value=-0.1}
         ]}}}))) {enabled=false, module=ScaleMeshSize002}
   &ScaleMeshSize(
    Initial Mesh Scale:Vector3f=$StackContext.Scale
    Scale Factor=&MakeVector(
     X=1
     Y=1
     Z=&FloatFromCurve(
      FloatCurve=DataInterface() {propertyValues {Curve {keys [
          {tangentMode=RCTM_User, time=0, value=1}
          <at 0.9 0.95 RCIM_Constant>
          <at 0.95 1.6 RCIM_Constant>
          <at 0.98 0.52 RCIM_Constant>
          <at 1 -0.1>
         ]}}}
      CurveIndex=&Subtract_Float(
       A=$Particles.NormalizedAge
       B=&ReturnNormalizedExecIndex(Normalized Index Scale=0.15, Normalized Index Includes One=false)
      )
     )
    )
   ) {module=ScaleMeshSize001}
   /SetVariables(Particles.MeshOrientation=&MakeCustomQuaternionFromBool(
     Bool:NiagaraBool=$System.SteppedTickRate.RefreshTick
     True Quat:Quat4f=$Particles.SkeletalMeshLocation.SampledRotation
     False Quat:Quat4f=$Particles.MeshOrientation
    )) {enabled=false, isSetParameters=true, module=SetVariables_#9}
   /SetVariables(Particles.CustomMeshIndex:NiagaraInt32=&MakeCustomIntFromBool(
     Bool=$User.TriggerKill
     True INT=&Add_Integer(
      A=&Multiply_Int(A=$Emitter.FramePerShape, B=$Emitter.ShapeNumber)
      B=&ModuloInt(
       A=&MakeIntFromFloat(Float=&Multiply_Float(A=$Engine.Time, B=8))
       B=$Emitter.FramePerShape
      )
     )
     False INT=&ModuloInt(
      A=&Add_Integer(
       A=&ModuloInt(
        A=&MakeIntFromFloat(Float=&Multiply_Float(A=$Engine.Time, B=6))
        B=$Emitter.FramePerShape
       )
       B=&Multiply_Int(
        A=&ModuloInt(
         A=&Add_Integer(
          A=&MakeIntFromFloat(Float=&Multiply_Float(A=$Engine.Time, B=1))
          B=&ReturnExecIndex()
         )
         B=$Emitter.ShapeNumber
        )
        B=$Emitter.FramePerShape
       )
      )
      B=&Multiply_Int(A=$Emitter.FramePerShape, B=$Emitter.ShapeNumber)
     )
    )) {isSetParameters=true, module=SetVariables_#10}
   &ScaleColor(
    Color Value To Scale=$Particles.Color
    Scale RGB=&Multiply_VectorByFloat(Vector=<1 1 1>, Float=1)
    ScaleA=true
    ScaleRGB=true
   ) {module=ScaleColor001}
   &Color(Color=&SelectLinearColorFromArray(
     Array Sampling Mode=::NewEnumerator1
     Direct Array Index=&Add_Integer(
      A=&MakeIntFromFloat(Float=&Multiply_Float(A=$Engine.Time, B=1))
      B=&ReturnExecIndex()
     )
     Direct Array Mode=::NewEnumerator1
     Color Selection Array=DataInterface() {propertyValues {ColorData [<rgba 1 0.1587 0.2417 1>, <rgba 0.949 0.3176 1 1>, <rgba 0.342 0.5018 1 1>, <rgba 1 0.8519 0.2575 1>, <rgba 0.3783 1 0.3917 1>]}}
     Random Seed=?
    )) {enabled=false}
  ]
  rendererClasses [&NiagaraMeshRendererProperties]
  renderers [
   {rendererClass=&NiagaraMeshRendererProperties, rendererIndex=0}
   {rendererClass=&NiagaraMeshRendererProperties, rendererIndex=1}
   {rendererClass=&NiagaraMeshRendererProperties, rendererIndex=2}
  ]
 }
 Stage1001 {
  emitterSpawnScript [/SetVariables(Emitter.FramePerShape=8, Emitter.ShapeNumber=1) {isSetParameters=true, module=SetVariables_#5}]
  emitterUpdateScript [
   &EmitterState(Life Cycle Mode=::NewEnumerator1, Loop Behavior=::NewEnumerator1, Loop Duration Mode=::NewEnumerator1, UseLoopDelay=false)
   &SpawnBurst_Instantaneous(
    Spawn Count=&MakeCustomIntFromBool(Bool=&SetBoolByIntComparison(A=$User.KillStage, B=1, Comparison Type=::NewEnumerator4), True INT=1, False INT=0)
    Spawn Group=1
    Use Spawn Probability=false
    UseLoopCountLimit=false
   ) {module=SpawnBurst_Instantaneous001}
   &NM_ToonShaderParameters(RimLight_Color:LinearColor=<rgba 0.1615 0.1615 0.1615 1>)
   &NM_OutlineParameters()
  ]
  particleSpawnScript [
   &FN_SetSpawnGroup(Number of Spawn Groups=2)
   &InitializeParticle(
    Lifetime=2
    Mesh Scale Mode=::NewEnumerator4
    Sprite Size Mode=::NewEnumerator3
    Sprite UV Mode=::NewEnumerator3
    Uniform Sprite Size=25
    UsePositionOffset=false
    Write Material Random=true
   ) {module=InitializeParticle001}
   /SetVariables(Particles.CustomMeshIndex:NiagaraInt32=&RandomRangeInt(Minimum=0, Maximum=3)) {isSetParameters=true, module=SetVariables_#6}
   /SetVariables(Particles.VisibilityTag=&MakeIntFromFloat(Float=$User.VariantID)) {isSetParameters=true}
   /SetVariables() {isSetParameters=true}
   /SetVariables(Particles.Color=&MultiplyLinearColor(
     Linear Color=$Particles.Color
     Scale Factor=&MakeVector4_FromLinearColor(LinearColor=&MakeLinearColorFromFloat(Value:NiagaraFloat=$System.TimeOfDayBrightnessMultiplier))
    )) {isSetParameters=true}
  ]
  particleUpdateScript [
   &ParticleState()
   &ScaleMeshSize(Scale Factor=&MakeVector(
     X=1
     Y=&FloatFromCurve(FloatCurve=DataInterface() {propertyValues {Curve {keys [<at 0 2 RCIM_Constant>,<at 0.05 1.5 RCIM_Constant>,<at 0.1 1 RCIM_Constant>]}}})
     Z=1
    )) {enabled=false}
   /SetVariables(Particles.Position=$Particles.SkeletalMeshLocation.SampledPosition) {enabled=false, isSetParameters=true, module=SetVariables_#7}
   /SetVariables(Particles.MeshOrientation=&MakeCustomQuaternionFromBool(
     Bool:NiagaraBool=$System.SteppedTickRate.RefreshTick
     True Quat:Quat4f=$Particles.SkeletalMeshLocation.SampledRotation
     False Quat:Quat4f=$Particles.MeshOrientation
    )) {enabled=false, isSetParameters=true, module=SetVariables_#9}
   /SetVariables(Particles.CustomMeshIndex:NiagaraInt32=&MakeCustomIntFromBool(
     Bool=$User.TriggerKill
     True INT=&Add_Integer(
      A=&Multiply_Int(A=$Emitter.FramePerShape, B=$Emitter.ShapeNumber)
      B=&ModuloInt(
       A=&MakeIntFromFloat(Float=&Multiply_Float(A=$Engine.Time, B=8))
       B=$Emitter.FramePerShape
      )
     )
     False INT=&ModuloInt(
      A=&Add_Integer(
       A=&ModuloInt(
        A=&MakeIntFromFloat(Float=&Multiply_Float(A=$Engine.Time, B=6))
        B=$Emitter.FramePerShape
       )
       B=&Multiply_Int(
        A=&ModuloInt(
         A=&Add_Integer(
          A=&MakeIntFromFloat(Float=&Multiply_Float(A=$Engine.Time, B=1))
          B=&ReturnExecIndex()
         )
         B=$Emitter.ShapeNumber
        )
        B=$Emitter.FramePerShape
       )
      )
      B=&Multiply_Int(A=$Emitter.FramePerShape, B=$Emitter.ShapeNumber)
     )
    )) {isSetParameters=true, module=SetVariables_#10}
   /SetVariables(
    Particles.Position=&AddVectorToPosition(
     Position=$Particles.Initial.Position
     Vector=&Multiply_Vector(
      A=&VectorFromFloat(Value=0.2)
      B=&SelectVectorFromArray(
       Direct Array Index=&ModuloInt(A=&MakeIntFromFloat(Float=&Multiply_Float(A=$Engine.Time, B=6)), B=6)
       Vector Selection Array=DataInterface() {propertyValues {FloatData [<1 0 0>, <0 -0.5 -0.5>, <-1 0.5 0.5>, <0.5 2 1>, <0.1 -0.4 0.2>, <0.35 0.7 -0.4>]}}
      )
     )
    )
    Particles.Scale=&Multiply_Vector(
     A=$Particles.Initial.Scale
     B=&SelectVectorFromArray(
      Direct Array Index=&ModuloInt(A=&MakeIntFromFloat(Float=&Multiply_Float(A=$Engine.Time, B=6)), B=3)
      Vector Selection Array=DataInterface() {propertyValues {FloatData [<1 1 1>, <0.99 1 1.05>, <1 1.02 0.99>]}}
     )
    )
   ) {isSetParameters=true}
   &ScaleMeshSize(
    Initial Mesh Scale:Vector3f=$StackContext.Scale
    Scale Factor=&MakeVector(
     X=1
     Y=1
     Z=&FloatFromCurve(
      FloatCurve=DataInterface() {propertyValues {Curve {keys [
          {tangentMode=RCTM_User, time=0, value=1}
          <at 0.9 0.95 RCIM_Constant>
          <at 0.95 1.6 RCIM_Constant>
          <at 0.98 0.52 RCIM_Constant>
          <at 1 -0.1>
         ]}}}
      CurveIndex=&Subtract_Float(
       A=$Particles.NormalizedAge
       B=&ReturnNormalizedExecIndex(Normalized Index Scale=0.15, Normalized Index Includes One=false)
      )
     )
    )
   ) {module=ScaleMeshSize001}
   &ScaleColor(
    Color Value To Scale=$Particles.Color
    Scale RGB=&Multiply_VectorByFloat(Vector=<1 1 1>, Float=1)
    ScaleA=true
    ScaleRGB=true
   ) {module=ScaleColor001}
   &Color(Color=&SelectLinearColorFromArray(
     Array Sampling Mode=::NewEnumerator1
     Direct Array Index=&Add_Integer(
      A=&MakeIntFromFloat(Float=&Multiply_Float(A=$Engine.Time, B=1))
      B=&ReturnExecIndex()
     )
     Direct Array Mode=::NewEnumerator1
     Color Selection Array=DataInterface() {propertyValues {ColorData [<rgba 1 0.1587 0.2417 1>, <rgba 0.949 0.3176 1 1>, <rgba 0.342 0.5018 1 1>, <rgba 1 0.8519 0.2575 1>, <rgba 0.3783 1 0.3917 1>]}}
     Random Seed=?
    )) {enabled=false}
  ]
  rendererClasses [&NiagaraMeshRendererProperties]
  renderers [
   {rendererClass=&NiagaraMeshRendererProperties, rendererIndex=0}
   {rendererClass=&NiagaraMeshRendererProperties, rendererIndex=1}
   {rendererClass=&NiagaraMeshRendererProperties, rendererIndex=2}
  ]
 }
 Stage2_Plus {
  emitterUpdateScript [
   &EmitterState()
   &SpawnBurst_Instantaneous(
    Spawn Count=&MakeCustomIntFromBool(Bool=&SetBoolByIntComparison(A=$User.KillStage, B=2, Comparison Type=::NewEnumerator4), True INT=2, False INT=0)
    Spawn Group=1
    Use Spawn Probability=false
    UseLoopCountLimit=false
   )
   &NM_ToonShaderParameters(Lighting_Hardness=0, RimLight_Color:LinearColor=<rgba 1 1 1 1>, RimLight_Width=0)
   &NM_OutlineParameters(OutlineWidth=0.5)
  ]
  particleSpawnScript [
   &PartitionParticles()
   &FN_SetSpawnGroup(Number of Spawn Groups=2) {enabled=false}
   &InitializeParticle(
    Color=&SelectLinearColorFromArray(
     Array Sampling Mode=::NewEnumerator1
     Direct Array Index=&MakeIntFromFloat(Float=$User.VariantID)
     Color Selection Array=DataInterface() {propertyValues {ColorData [<rgba 0.12 0.281 1.5 1>, <rgba 0.1896 1.8739 2 1>]}}
     Random Seed=?
    )
    Lifetime=2
    Mesh Scale Mode=::NewEnumerator3
    Mesh Uniform Scale=&SelectFloatFromArray(
     Direct Array Index=&ReturnExecIndex()
     Float Selection Array=DataInterface() {propertyValues {FloatData [1.2,0.8,0]}}
    )
    Position Offset=&Add_Vector(
     A=<0 0 0>
     B=&SelectVectorFromArray(
      Direct Array Index=&ReturnExecIndex()
      Vector Selection Array=DataInterface() {propertyValues {FloatData [<-5 0 14>, <-10 1 10>, <8 -17 -10>]}}
     )
    )
    Position Offset Coordinate Space=::Local
    Sprite UV Mode=::NewEnumerator3
    UsePositionOffset=true
    Write Material Random=true
   ) {module=InitializeParticle001}
   /SetVariables() {isSetParameters=true, module=SetVariables_#6}
   &InitialMeshOrientation(Mesh Orientation Mode=::NewEnumerator7, Rotation=<0.35 0 0>, Use Rotation Vector=true) {enabled=false}
   /SetVariables(Particles.Color=&MultiplyLinearColor(
     Linear Color=$Particles.Color
     Scale Factor=&MakeVector4_FromLinearColor(LinearColor=&MakeLinearColorFromFloat(Value:NiagaraFloat=$System.TimeOfDayBrightnessMultiplier))
    )) {isSetParameters=true}
  ]
  particleUpdateScript [
   &ParticleState()
   &ScaleMeshSize(Scale Factor=&VectorFromFloat(Value=&FloatFromCurve(
      FloatCurve=DataInterface() {propertyValues {Curve {keys [<at 0 1.7 RCIM_Constant>,<at 0.1 1.15 RCIM_Constant>,<at 0.25 1 RCIM_Constant>]}}}
      CurveIndex=&Subtract_Float(A=$Particles.NormalizedAge, B=&ReturnNormalizedExecIndex(Normalized Index Scale=0.01))
     ))) {module=ScaleMeshSize003}
   &ScaleMeshSize(Scale Factor=&VectorFromFloat(Value=&FloatFromCurve(FloatCurve=DataInterface() {propertyValues {Curve {keys [<at 0 0 RCIM_Constant>,<at 0.1 1.15 RCIM_Constant>,<at 0.25 1 RCIM_Constant>]}}}))) {enabled=false}
   &ScaleMeshSize(Scale Factor=&Multiply_Vector(
     A=$Particles.Initial.Scale
     B=&VectorFromFloat(Value=&RemapRange(
       Input Value To Remap=&Ceil_Float(Float=&Sine(Normalized Angle=$Emitter.Age, Period=0.25, Scale=0.02, Bias=1))
       Input Min=0
       Input Max=2
       Output Min=0.8
       Output Max=1
      ))
    )) {module=ScaleMeshSize001}
   /SetVariables(Particles.CustomMeshIndex:NiagaraInt32=&ModuloInt(A=&MakeIntFromFloat(Float=&Multiply_Float(A=$Engine.Time, B=8)), B=8)) {enabled=false, isSetParameters=true, module=SetVariables_#10}
   &ScaleColor(
    Color Value To Scale=$Particles.Color
    Scale RGB=&Multiply_VectorByFloat(Vector=<1 1 1>, Float=1)
    ScaleA=true
    ScaleRGB=true
   ) {module=ScaleColor001}
   &Color(Color=&SelectLinearColorFromArray(
     Array Sampling Mode=::NewEnumerator1
     Direct Array Index=&MakeIntFromFloat(Float=&Multiply_Float(A=$Engine.Time, B=0.333))
     Direct Array Mode=::NewEnumerator1
     Color Selection Array=DataInterface() {propertyValues {ColorData [<rgba 1 0.1587 0.2417 1>, <rgba 0.9506 0.3182 1 1>, <rgba 0.342 0.5018 1 1>, <rgba 1 0.8519 0.2575 1>, <rgba 0.3783 1 0.3917 1>]}}
     Random Seed=?
    )) {enabled=false}
   &self:ScratchModule(ReturnExecIndex:NiagaraInt32=&ReturnExecIndex(), StageSelector:NiagaraInt32=$User.KillStage) {enabled=false, module=ScratchModule}
   &ScaleColor(ScaleA=true, ScaleRGB=true)
   /SetVariables(
    Particles.Position=&AddVectorToPosition(
     Position=$Particles.Initial.Position
     Vector=&Multiply_Vector(
      A=&VectorFromFloat(Value=0.8)
      B=&SelectVectorFromArray(
       Direct Array Index=&ModuloInt(A=&MakeIntFromFloat(Float=&Multiply_Float(A=$Engine.Time, B=6)), B=6)
       Vector Selection Array=DataInterface() {propertyValues {FloatData [<0.2 0.2 0>, <0.5 0.4 -0.5>, <-0.1 -0.4 0.5>, <0.4 -1 1>, <0.3 0.7 0.2>, <0.8 0.7 -0.4>]}}
      )
     )
    )
    Particles.Scale=&Multiply_Vector(
     A=$Particles.Initial.Scale
     B=&SelectVectorFromArray(
      Direct Array Index=&ModuloInt(A=&MakeIntFromFloat(Float=&Multiply_Float(A=$Engine.Time, B=6)), B=3)
      Vector Selection Array=DataInterface() {propertyValues {FloatData [<0.8 1 1>, <1 1 1.05>, <1.1 1.05 0.99>]}}
     )
    )
   ) {isSetParameters=true}
   &ScaleMeshSize(
    Initial Mesh Scale:Vector3f=$StackContext.Scale
    Scale Factor=&VectorFromFloat(Value=&FloatFromCurve(
      FloatCurve=DataInterface() {propertyValues {Curve {keys [{tangentMode=RCTM_User, time=0, value=1}, <at 1.2 1 RCIM_Constant>, <at 1.25 1.1 RCIM_Constant>, <at 1.35 0.7 RCIM_Constant>, <at 1.42 0>]}}}
      CurveIndex=&Subtract_Float(A=$System.Age, B=&ReturnNormalizedExecIndex(Normalized Index Scale=0.2))
     ))
   ) {module=ScaleMeshSize002}
  ]
  rendererClasses [&NiagaraMeshRendererProperties]
  renderers [
   {rendererClass=&NiagaraMeshRendererProperties, rendererIndex=0}
   {rendererClass=&NiagaraMeshRendererProperties, rendererIndex=1}
  ]
 }
 Stage4_Plus001 {
  emitterUpdateScript [
   &EmitterState()
   &SpawnBurst_Instantaneous(
    Spawn Count=&MakeCustomIntFromBool(Bool=&SetBoolByIntComparison(A=$User.KillStage, B=3, Comparison Type=::NewEnumerator4), True INT=2, False INT=0)
    Spawn Group=1
    Use Spawn Probability=false
    UseLoopCountLimit=false
   )
   &NM_ToonShaderParameters(Lighting_Hardness=1, RimLight_Color:LinearColor=<rgba 1 1 1 1>, RimLight_Width=0)
   &NM_OutlineParameters(OutlineWidth=0.5)
  ]
  particleSpawnScript [
   &PartitionParticles()
   &InitializeParticle(
    Color=&SelectLinearColorFromArray(
     Array Sampling Mode=::NewEnumerator1
     Direct Array Index=&MakeIntFromFloat(Float=$User.VariantID)
     Color Selection Array=DataInterface() {propertyValues {ColorData [<rgba 0.12 0.281 1.5 1>, <rgba 0.1896 1.8739 2 1>]}}
     Random Seed=?
    )
    Lifetime=2
    Mesh Scale Mode=::NewEnumerator3
    Mesh Uniform Scale=&SelectFloatFromArray(
     Direct Array Index=&ReturnExecIndex()
     Float Selection Array=DataInterface() {propertyValues {FloatData [1.5,0.8,0.9]}}
    )
    Position Offset=&Add_Vector(
     A=<0 0 0>
     B=&SelectVectorFromArray(
      Direct Array Index=&ReturnExecIndex()
      Vector Selection Array=DataInterface() {propertyValues {FloatData [<-15 0 20>, <-8 1 14>, <8 -17 -10>]}}
     )
    )
    Position Offset Coordinate Space=::Local
    Sprite UV Mode=::NewEnumerator3
    UsePositionOffset=true
    Write Material Random=true
   ) {module=InitializeParticle001}
   /SetVariables() {isSetParameters=true, module=SetVariables_#6}
   /SetVariables(Particles.Color=&MultiplyLinearColor(
     Linear Color=$Particles.Color
     Scale Factor=&MakeVector4_FromLinearColor(LinearColor=&MakeLinearColorFromFloat(Value:NiagaraFloat=$System.TimeOfDayBrightnessMultiplier))
    )) {isSetParameters=true}
  ]
  particleUpdateScript [
   &ParticleState()
   &ScaleMeshSize(Scale Factor=&VectorFromFloat(Value=&FloatFromCurve(
      FloatCurve=DataInterface() {propertyValues {Curve {keys [<at 0 1.7 RCIM_Constant>,<at 0.1 1.15 RCIM_Constant>,<at 0.25 1 RCIM_Constant>]}}}
      CurveIndex=&Subtract_Float(A=$Particles.NormalizedAge, B=&ReturnNormalizedExecIndex(Normalized Index Scale=0.01))
     ))) {module=ScaleMeshSize003}
   /SetVariables(
    Particles.Position=&AddVectorToPosition(
     Position=$Particles.Initial.Position
     Vector=&Multiply_Vector(
      A=&VectorFromFloat(Value=0.2)
      B=&SelectVectorFromArray(
       Direct Array Index=&ModuloInt(A=&MakeIntFromFloat(Float=&Multiply_Float(A=$Engine.Time, B=6)), B=6)
       Vector Selection Array=DataInterface() {propertyValues {FloatData [<1 0 0>, <0 -0.5 -0.5>, <-1 0.5 0.5>, <0.5 2 1>, <0.1 -0.4 0.2>, <0.35 0.7 -0.4>]}}
      )
     )
    )
    Particles.Scale=&Multiply_Vector(
     A=$Particles.Initial.Scale
     B=&SelectVectorFromArray(
      Direct Array Index=&ModuloInt(
       A=&MakeIntFromFloat(Float=&Multiply_Float(A=&Subtract_Float(A=$Engine.Time, B=&ReturnNormalizedExecIndex()), B=6))
       B=3
      )
      Vector Selection Array=DataInterface() {propertyValues {FloatData [<0.8 1 1>, <0.99 1 1.05>, <1.2 1.02 0.99>]}}
     )
    )
   ) {isSetParameters=true}
   &ScaleMeshSize(
    Initial Mesh Scale:Vector3f=$StackContext.Scale
    Scale Factor=&VectorFromFloat(Value=&FloatFromCurve(
      FloatCurve=DataInterface() {propertyValues {Curve {keys [{tangentMode=RCTM_User, time=0, value=1}, <at 0.7 0.95 RCIM_Constant>, <at 0.75 1.6 RCIM_Constant>, <at 0.78 0.52 RCIM_Constant>, <at 0.8 0>]}}}
      CurveIndex=&Subtract_Float(
       A=$Particles.NormalizedAge
       B=&ReturnNormalizedExecIndex(Normalized Index Scale=0.15, Normalized Index Includes One=false)
      )
     ))
   ) {module=ScaleMeshSize005}
   &ScaleColor(
    Color Value To Scale=$Particles.Color
    Scale RGB=&Multiply_VectorByFloat(Vector=<1 1 1>, Float=1)
    ScaleA=true
    ScaleRGB=true
   ) {module=ScaleColor001}
   &ScaleColor(ScaleA=true, ScaleRGB=true)
  ]
  rendererClasses [&NiagaraMeshRendererProperties]
  renderers [
   {rendererClass=&NiagaraMeshRendererProperties, rendererIndex=0}
   {rendererClass=&NiagaraMeshRendererProperties, rendererIndex=1}
  ]
 }
 Stage4_Stars {
  emitterUpdateScript [
   &EmitterState()
   &SpawnBurst_Instantaneous(
    Spawn Count=&MakeCustomIntFromBool(Bool=&SetBoolByIntComparison(A=$User.KillStage, B=3, Comparison Type=::NewEnumerator4), True INT=3, False INT=0)
    Spawn Group=1
    Use Spawn Probability=false
    UseLoopCountLimit=false
   )
   &NM_ToonShaderParameters(HighLight_Size=0.1, HighLight_Softness=0.001, Lighting_Hardness=0, RimLight_Color:LinearColor=<rgba 1 1 1 1>, RimLight_Width=0)
   &NM_OutlineParameters(OutlineWidth=0.5)
  ]
  particleSpawnScript [
   &PartitionParticles()
   &FN_SetSpawnGroup(Number of Spawn Groups=2) {enabled=false}
   &InitializeParticle(
    Color=&MultiplyLinearColor(
     Linear Color=&SelectLinearColorFromArray(
      Array Sampling Mode=::NewEnumerator1
      Direct Array Index=&MakeIntFromFloat(Float=$User.VariantID)
      Color Selection Array=DataInterface() {propertyValues {ColorData [<rgba 5 3.4385 0 1>, <rgba 1.8938 0.075 2.5 1>]}}
      Random Seed=?
     )
     Scale Factor=&MakeVector4_FromLinearColor(LinearColor=<rgba 1 1 1 1>)
    )
    Lifetime=1.7
    Mesh Scale Mode=::NewEnumerator3
    Mesh Uniform Scale=&SelectFloatFromArray(
     Direct Array Index=&ReturnExecIndex()
     Float Selection Array=DataInterface() {propertyValues {FloatData [1,1.3]}}
    )
    Position Offset=&Add_Vector(
     A=<0 0 0>
     B=&SelectVectorFromArray(
      Direct Array Index=&ReturnExecIndex()
      Vector Selection Array=DataInterface() {propertyValues {FloatData [<15 0 10>, <10 1 17>, <-5 -20 -10>]}}
     )
    )
    Position Offset Coordinate Space=::Local
    Sprite UV Mode=::NewEnumerator3
    UsePositionOffset=true
    Write Material Random=true
   ) {module=InitializeParticle001}
   /SetVariables() {isSetParameters=true, module=SetVariables_#6}
   &InitialMeshOrientation(Mesh Orientation Mode=::NewEnumerator7, Rotation=<0.35 0 0>, Use Rotation Vector=true) {enabled=false}
   /SetVariables(Particles.Color=&LerpLinearColors(
     StartColor=$Particles.Color
     EndColor=&SelectLinearColorFromArray(
      Array Sampling Mode=::NewEnumerator1
      Direct Array Index=&ReturnExecIndex()
      Color Selection Array=DataInterface() {propertyValues {ColorData [<rgba 2 0.4 1.2442 1>, <rgba 2 0.2 1.1498 1>, <rgba 2 0.56 0.4 1>]}}
      Random Seed=?
     )
     LerpFactor=$User.VariantID
    )) {isSetParameters=true}
   /SetVariables(Particles.Color=&MultiplyLinearColor(
     Linear Color=$Particles.Color
     Scale Factor=&MakeVector4_FromLinearColor(LinearColor=&MakeLinearColorFromFloat(Value:NiagaraFloat=$System.TimeOfDayBrightnessMultiplier))
    )) {isSetParameters=true}
  ]
  particleUpdateScript [
   &ParticleState()
   &ScaleMeshSize(Scale Factor=&VectorFromFloat(Value=&FloatFromCurve(
      FloatCurve=DataInterface() {propertyValues {Curve {keys [<at 0 2 RCIM_Constant>,<at 0.1 1.15 RCIM_Constant>,<at 0.25 1 RCIM_Constant>]}}}
      CurveIndex=&Subtract_Float(A=$Particles.NormalizedAge, B=&ReturnNormalizedExecIndex(Normalized Index Scale=0.01))
     )))
   &ScaleMeshSize(Scale Factor=&Multiply_Vector(
     A=$Particles.Initial.Scale
     B=&VectorFromFloat(Value=&RemapRange(
       Input Value To Remap=&Ceil_Float(Float=&Sine(Normalized Angle=$Emitter.Age, Period=0.25, Scale=0.02, Bias=1))
       Input Min=0
       Input Max=2
       Output Min=0.8
       Output Max=1
      ))
    )) {module=ScaleMeshSize001}
   /SetVariables(
    Particles.Position=&AddVectorToPosition(
     Position=$Particles.Initial.Position
     Vector=&Multiply_Vector(
      A=&VectorFromFloat(Value=0.5)
      B=&SelectVectorFromArray(
       Direct Array Index=&MakeIntFromFloat(Float=&ModuloFloat(A=&Multiply_Float(A=$System.Stage4JitterSpeed, B=1), B=3))
       Vector Selection Array=DataInterface() {propertyValues {FloatData [<0 0 0>, <1 5 0.5>, <-1 0.5 -0.5>]}}
      )
     )
    )
    Particles.Scale=&Multiply_Vector(
     A=$Particles.Initial.Scale
     B=&SelectVectorFromArray(
      Direct Array Index=&MakeIntFromFloat(Float=&ModuloFloat(A=&Multiply_Float(A=$System.Stage4JitterSpeed, B=1), B=3))
      Vector Selection Array=DataInterface() {propertyValues {FloatData [<1 1 1>, <1.12 1.2 1.3>, <1 1.2 1.2>]}}
     )
    )
   ) {enabled=false, isSetParameters=true}
   /SetVariables(
    Particles.Position=&AddVectorToPosition(
     Position=$Particles.Initial.Position
     Vector=&Multiply_Vector(
      A=&VectorFromFloat(Value=0.2)
      B=&SelectVectorFromArray(
       Direct Array Index=&ModuloInt(A=&MakeIntFromFloat(Float=&Multiply_Float(A=$Engine.Time, B=6)), B=6)
       Vector Selection Array=DataInterface() {propertyValues {FloatData [<1 0 0>, <0 -0.5 -0.5>, <-1 0.5 0.5>, <0.5 2 1>, <0.1 -0.4 0.2>, <0.35 0.7 -0.4>]}}
      )
     )
    )
    Particles.Scale=&Multiply_Vector(
     A=$Particles.Initial.Scale
     B=&SelectVectorFromArray(
      Direct Array Index=&ModuloInt(A=&MakeIntFromFloat(Float=&Multiply_Float(A=$Engine.Time, B=6)), B=3)
      Vector Selection Array=DataInterface() {propertyValues {FloatData [<1 1 1>, <0.99 1 1.05>, <1 1.02 0.99>]}}
     )
    )
   ) {isSetParameters=true}
   &ScaleMeshSize(Scale Factor=&VectorFromFloat(Value=&FloatFromCurve(
      FloatCurve=DataInterface() {propertyValues {Curve {keys [{tangentMode=RCTM_User, time=0, value=1}, <at 0.7 1 RCIM_Constant>, <at 0.8 0.6 RCIM_Constant>, <at 0.9 -0.1>]}}}
      CurveIndex=&Subtract_Float(
       A=$Particles.NormalizedAge
       B=&ReturnNormalizedExecIndex(Normalized Index Scale=0.2, Normalized Index Includes One=false)
      )
     ))) {module=ScaleMeshSize002}
   /SetVariables(Particles.CustomMeshIndex:NiagaraInt32=&ModuloInt(A=&MakeIntFromFloat(Float=&Multiply_Float(A=$Engine.Time, B=8)), B=8)) {enabled=false, isSetParameters=true, module=SetVariables_#10}
   &ScaleColor(
    Color Value To Scale=$Particles.Color
    Scale RGB=&Multiply_VectorByFloat(Vector=<1 1 1>, Float=1)
    ScaleA=true
    ScaleRGB=true
   ) {module=ScaleColor001}
   &Color(Color=&SelectLinearColorFromArray(
     Array Sampling Mode=::NewEnumerator1
     Direct Array Index=&MakeIntFromFloat(Float=&Multiply_Float(A=$Engine.Time, B=0.333))
     Direct Array Mode=::NewEnumerator1
     Color Selection Array=DataInterface() {propertyValues {ColorData [<rgba 1 0.1587 0.2417 1>, <rgba 0.9506 0.3182 1 1>, <rgba 0.342 0.5018 1 1>, <rgba 1 0.8519 0.2575 1>, <rgba 0.3783 1 0.3917 1>]}}
     Random Seed=?
    )) {enabled=false}
   &self:ScratchModule(ReturnExecIndex:NiagaraInt32=&ReturnExecIndex(), StageSelector:NiagaraInt32=$User.KillStage) {enabled=false, module=ScratchModule}
   &ScaleColor(ScaleA=true, ScaleRGB=true)
   /SetVariables(Particles.Position=&AddVectorToPosition(
     Position=$Particles.Initial.Position
     Vector=&Multiply_Vector(
      A=&VectorFromFloat(Value=0.5)
      B=&SelectVectorFromArray(
       Direct Array Index=&MakeIntFromFloat(Float=&ModuloFloat(A=&Multiply_Float(A=$Engine.Time, B=8), B=3))
       Vector Selection Array=DataInterface() {propertyValues {FloatData [<0 -0.5 -0.5>, <0 0 0>, <0 0.5 0.5>]}}
      )
     )
    )) {enabled=false, isSetParameters=true}
  ]
  rendererClasses [&NiagaraMeshRendererProperties]
  renderers [
   {rendererClass=&NiagaraMeshRendererProperties, rendererIndex=0}
   {rendererClass=&NiagaraMeshRendererProperties, rendererIndex=1}
  ]
 }
}
summary {
 systemName=NS_Backpack_Dazzle_Burst02
 emitters [
  {emitterName=Stage4_Stars, rendererClasses [&NiagaraMeshRendererProperties]}
  {emitterName=Stage4_Plus001, rendererClasses [&NiagaraMeshRendererProperties]}
  {emitterName=Stage2_Plus, rendererClasses [&NiagaraMeshRendererProperties]}
  {emitterName=Stage1001, rendererClasses [&NiagaraMeshRendererProperties]}
  {emitterName=NoFixedRate001, rendererClasses [&NiagaraMeshRendererProperties]}
  {emitterName=NoFixedRate002, rendererClasses [&NiagaraMeshRendererProperties]}
 ]
 userVariables [
  User.bIsFrontEnd {type=NiagaraBool}
  User.bIsFrontEndPreview {type=NiagaraBool}
  User.KillStage {type=NiagaraInt32, defaultValue=1}
  User.MeshColors {
   type=NiagaraDataInterfaceArrayColor
   description=""
   defaultValue {
    struct=NiagaraExt_VariableValue_DataInterface
    value {
     dataInterface=&self:NiagaraDataInterfaceArrayColor_0
     dataInterfaceClass=&NiagaraDataInterfaceArrayColor
    }
   }
  }
  User.SkeletalMesh {
   type=NiagaraDataInterfaceSkeletalMesh
   defaultValue {
    struct=NiagaraExt_VariableValue_DataInterface
    value {dataInterface=&self:NiagaraDataInterfaceSkeletalMesh_0}
   }
  }
  User.Stage1JitterSpeedMultiplier {type=NiagaraFloat, defaultValue=14, description=""}
  User.Stage4JitterSpeedMultiplier {type=NiagaraFloat, defaultValue=14, description=""}
  User.TriggerKill {type=NiagaraBool}
  User.VariantID {type=NiagaraFloat, defaultValue=1, description=""}
 ]
}
