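# Physics asset for the F_MED_Dazzle backpack cosmetic, defining collision bodies and joints for its dangling parts.
# A capsule-and-box backpack body anchors a handle whose joint locks twist and swing2, leaving one swing axis.
# A chain of five small "ruber" capsules hangs from the handle, each linked to the previous with modest swing limits.
# A separate large capsule on the pack bone has no constraint.
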PhysicsAsset /BRCosmetics/Accessories/FORT_Backpacks/Backpack_F_MED_Dazzle/Meshes/F_MED_Dazzle_Pack_Physics
bodies {
 dyn_backpack [
  {
   shapeType=Capsule
   length=5.7591
   radius=8.8499
   shapeName=dyn_backpack_capsule
   center=<-0.0353 -1.2161 2.522>
   rotation=<rot 4.358 177.3806 -116.9651>
  }
  {shapeType=Box, extentX=18.7764, extentY=17.2925, extentZ=11.4705, shapeName=dyn_backpack_box, rotation=<rot 0 -15.2699 27.5887>}
 ]
 dyn_handle [{
   shapeType=Capsule
   length=5.2145
   radius=1.9015
   shapeName=dyn_handle_capsule
   center=<7.0359 -0.5706 -0.5158>
   rotation=<rot 0 -67.5886 -90>
  }]
 dyn_ruber1 [{
   shapeType=Capsule
   length=0.5925
   radius=0.5786
   shapeName=dyn_ruber1_capsule
   center=<0.5459 0.0363 -0.056>
   rotation=<rot 0 -101.1655 -74.9646>
  }]
 dyn_ruber2 [{
   shapeType=Capsule
   length=0.8843
   radius=0.5101
   shapeName=dyn_ruber2_capsule
   center=<0.765 0.0749 -0.068>
   rotation=<rot 0 84.3314 -89.336>
  }]
 dyn_ruber3 [{
   shapeType=Capsule
   length=0.505
   radius=0.505
   shapeName=dyn_ruber3_capsule
   center=<0.4704 0.0344 0.0517>
   rotation=<rot 0 -105.9854 -83.5548>
  }]
 dyn_ruber4 [{
   shapeType=Capsule
   length=0.505
   radius=0.505
   shapeName=dyn_ruber4_capsule
   center=<0.3545 -0.0688 -0.0161>
   rotation=<rot 0 76.0672 -86.3922>
  }]
 dyn_ruber_end [{
   shapeType=Capsule
   length=0.505
   radius=0.505
   shapeName=dyn_ruber_end_capsule
   center=<0.2458 -0.0442 -0.087>
   rotation=<rot 12.3402 -90 -2.6411>
  }]
 pack [{shapeType=Capsule, length=13.3204, radius=12.7009, center=<1.1077 -7.4606 0.9243>, rotation=<rot 36.5595 -90 -120.4861>}]
}
constraints [
 {bone1Name=dyn_ruber1, bone2Name=dyn_handle, swing1LimitDegrees=15, swing2LimitDegrees=15, twistLimitDegrees=0}
 {bone1Name=dyn_ruber2, bone2Name=dyn_ruber1, swing1LimitDegrees=30, swing2LimitDegrees=30, twistLimitDegrees=0}
 {bone1Name=dyn_ruber3, bone2Name=dyn_ruber2, swing1LimitDegrees=30, swing2LimitDegrees=30, twistLimitDegrees=0}
 {bone1Name=dyn_ruber4, bone2Name=dyn_ruber3, swing1LimitDegrees=30, swing2LimitDegrees=30, twistLimitDegrees=0}
 {bone1Name=dyn_ruber_end, bone2Name=dyn_ruber4, swing1LimitDegrees=20, swing2LimitDegrees=20, twistLimitDegrees=0}
 {
  bone1Name=dyn_handle
  bone2Name=dyn_backpack
  swing1LimitDegrees=15
  swing2LimitDegrees=45
  swing2Motion=Locked
  twistLimitDegrees=45
  twistMotion=Locked
 }
]
