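# Physics asset for the antenna on the F_MED_Dazzle backpack cosmetic, with capsule bodies on the pack, backpack, antenna and antenna tip bones.
# The tip is angularly locked to the antenna, which joins dyn_backpack through a joint with only swing 2 left unlocked, so it pivots on a single axis.
# The larger pack capsule has no constraint to the other bodies.
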PhysicsAsset /BRCosmetics/Accessories/FORT_Backpacks/Backpack_F_MED_Dazzle/Meshes/F_MED_Dazzle_Pack_Physics_anten
bodies {
 dyn_anten [{
   shapeType=Capsule
   length=13.3964
   radius=0.5
   shapeName=dyn_anten_capsule
   center=<6.4021 0.3181 -0.0241>
   rotation=<rot 0 90.4067 -89.8781>
  }]
 dyn_anten_end [{shapeType=Capsule, length=0.505, radius=0.505, shapeName=dyn_anten_end_capsule, rotation=<rot 0 -90 0>}]
 dyn_backpack [{
   shapeType=Capsule
   length=5.7591
   radius=8.8499
   shapeName=dyn_backpack_capsule
   center=<-0.0353 -1.2161 2.522>
   rotation=<rot 4.358 177.3806 -116.9651>
  }]
 pack [{shapeType=Capsule, length=13.3204, radius=12.7009, center=<1.1077 -7.4606 0.9243>, rotation=<rot 36.5595 -90 -120.4861>}]
}
constraints [
 {
  bone1Name=dyn_anten_end
  bone2Name=dyn_anten
  swing1LimitDegrees=45
  swing1Motion=Locked
  swing2LimitDegrees=45
  swing2Motion=Locked
  twistLimitDegrees=45
  twistMotion=Locked
 }
 {
  bone1Name=dyn_anten
  bone2Name=dyn_backpack
  swing1LimitDegrees=45
  swing1Motion=Locked
  swing2LimitDegrees=5
  twistLimitDegrees=45
  twistMotion=Locked
 }
]
